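; Blueprint for the GlowDonut office sconce A environment prop, with glass and candle flame plane meshes and support for dynamic destruction.
; It overrides GetDynamicDestructionStaticMeshComponents to return three component arrays: the main static mesh alone, the flame plane with the glass, and the main mesh with the glass.
; The construction script only calls its parent, and the BeginPlay, overlap and Tick events are empty.
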(event EventBeginPlay)

(event Collision|EventActorBeginOverlap
  (params (OtherActor :type Actor%20Object%20Reference)))

(event EventTick
  (params (DeltaSeconds :type Float%20%28single-precision%29)))

(fn GetDynamicDestructionStaticMeshComponents
  (return (Utilities|Array|MakeArray (Variables|Debug|GetStaticMeshComponent)) (Utilities|Array|MakeArray (Variables|Default|GetSM_WildEstate_CandleFlamePlane) (Variables|Default|GetSM_GlowDonut_OfficeSconceGlass_A)) (Utilities|Array|MakeArray (Variables|Debug|GetStaticMeshComponent) (Variables|Default|GetSM_GlowDonut_OfficeSconceGlass_A))))

(fn ConstructionScript
  (|CallParentFunction))
